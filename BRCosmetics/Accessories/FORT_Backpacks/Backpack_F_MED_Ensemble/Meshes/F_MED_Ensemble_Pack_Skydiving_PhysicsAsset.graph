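# Physics asset for the skydiving version of the F_MED_Ensemble backpack cosmetic, set up so its three zipper pulls can swing.
# Small capsule bodies cover the pack, three base zip bones and three dynamic zip bones, with the dynamic capsules offset and rotated along their bones.
# Each dynamic zip is constrained to its matching base with narrow swing limits and almost no twist; the pack body has no constraint.
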PhysicsAsset /BRCosmetics/Accessories/FORT_Backpacks/Backpack_F_MED_Ensemble/Meshes/F_MED_Ensemble_Pack_Skydiving_PhysicsAsset
bodies {
 base_zip_A [{shapeType=Capsule, length=1, radius=0.5, shapeName=base_zip_A_capsule}]
 base_zip_B [{shapeType=Capsule, length=1, radius=0.5, shapeName=base_zip_B_capsule}]
 base_zip_C [{shapeType=Capsule, length=1, radius=0.5, shapeName=base_zip_C_capsule}]
 dyn_zip_A [{shapeType=Capsule, length=1, radius=0.5, shapeName=dyn_zip_A_capsule, center=<2 0 0>, rotation=<rot 90 0 0>}]
 dyn_zip_B [{shapeType=Capsule, length=1, radius=0.5, shapeName=dyn_zip_B_capsule, center=<2 0 0>, rotation=<rot 90 0 0>}]
 dyn_zip_C [{shapeType=Capsule, length=1, radius=0.5, shapeName=dyn_zip_C_capsule, center=<2 0 0>, rotation=<rot 90 0 0>}]
 pack [{shapeType=Capsule, length=1, radius=1, shapeName=pack_capsule}]
}
constraints [
 {bone1Name=dyn_zip_A, bone2Name=base_zip_A, swing1LimitDegrees=30, swing2LimitDegrees=25, twistLimitDegrees=1}
 {bone1Name=dyn_zip_B, bone2Name=base_zip_B, swing1LimitDegrees=20, swing2LimitDegrees=20, twistLimitDegrees=1}
 {bone1Name=dyn_zip_C, bone2Name=base_zip_C, swing1LimitDegrees=20, swing2LimitDegrees=20, twistLimitDegrees=1}
]
